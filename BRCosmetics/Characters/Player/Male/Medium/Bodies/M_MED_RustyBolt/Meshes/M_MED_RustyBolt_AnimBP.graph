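# Post-process style AnimBP for the Rusty Bolt cosmetic body, copying the attached parent mesh's pose and adding secondary motion to gear bones.
# AnimDynamics swings pouches, while constraints and additive bone transforms adjust plate, shoulder pads, shoulder volume and pocket.
# EventGraph derives those nodes' alphas from weapon type, skydive, emote and front-end states, plus arm, clavicle and thigh yaw remapped from reference-pose deltas.
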AnimBlueprint /BRCosmetics/Characters/Player/Male/Medium/Bodies/M_MED_RustyBolt/Meshes/M_MED_RustyBolt_AnimBP
target_skeleton=&/M_MED_RustyBolt_Skeleton
graphs [
 AnimGraph {nodes [
   Blend Poses by bool(
    BlendPose_0=Use cached pose 'Base_Cache'()
    BlendPose_1=Use cached pose 'ShoulderPad_Cache'()
    bActiveValue=Get FrontEnd_Idle_On() -> (FrontEnd_Idle_On:bool)
   ) {blendType=Linear}
   "Constraint - dyn_front_belt (PPPP)"!(
    ComponentPose
    Alpha=1
    ConstraintWeights_0=0.3552
    ConstraintWeights_1:float=0.4321
    ConstraintWeights_2:float=0.0494
    ConstraintWeights_3:float=0.1634
   ) -> (Pose) {
    alphaBoolBlend {blendOption=Linear}
    boneToModify {boneName=dyn_front_belt}
    constraintSetup [{targetBone {boneName=pelvis}}, {targetBone {boneName=spine_01}}, {targetBone {boneName=spine_02}}, {targetBone {boneName=spine_03}}]
    constraintWeights [1,1,1,1]
   }
   Local To Component(LocalPose=Use cached pose 'Base_Cache'())
   n1 = Anim Dynamics - Bone: dyn_back_pouch(ComponentPose=Local To Component(LocalPose=Copy Pose From Mesh() {bUseAttachedParent=true}), Alpha=1, ExternalForce="0,0,0") {
    angularDampingOverride=0.999
    angularSpringConstant=200000
    bAngularSpring=true
    bGravityOverrideInSimSpace=true
    bOverrideAngularDamping=true
    bUseGravityOverride=true
    gravityScale=0.35
    simulationSpace=RootRelative
    alphaBoolBlend {blendOption=Linear}
    boundBone {boneName=dyn_back_pouch}
    componentLinearAccScale=<1 1 1>
    componentLinearVelScale=<1 1 1>
    gravityOverride {z=-980}
    physicsBodyDefinitions [{
      sphereCollisionRadius=0
      boundBone {boneName=dyn_back_pouch}
      boxExtents=<150 80 90>
      constraintSetup {angularLimitsMax=<20 10 1>, angularLimitsMin=<-20 -10 -1>, angularTarget=<1 0 0>}
      localJointOffset=<-1 -4 -4>
     }]
   }
   n2 = Anim Dynamics - Bone: dyn_L_pouch(ComponentPose=@n1, Alpha=1, ExternalForce="0,0,0") {
    angularDampingOverride=0.9
    angularSpringConstant=200000
    bAngularSpring=true
    bGravityOverrideInSimSpace=true
    bOverrideAngularDamping=true
    bUseGravityOverride=true
    gravityScale=0.35
    simulationSpace=RootRelative
    alphaBoolBlend {blendOption=Linear}
    boundBone {boneName=dyn_L_pouch}
    componentLinearAccScale=<1 1 1>
    componentLinearVelScale=<1 1 1>
    gravityOverride {z=-980}
    physicsBodyDefinitions [{
      sphereCollisionRadius=0
      boundBone {boneName=dyn_L_pouch}
      boxExtents=<100 30 50>
      constraintSetup {angularLimitsMax=<3 3 0>, angularLimitsMin=<-3 -3 0>, angularTarget=<1 0 0>}
      localJointOffset=<-1 4 -4>
     }]
   }
   n3 = Anim Dynamics - Bone: dyn_R_pouch(ComponentPose=@n2, Alpha=1, ExternalForce="0,0,0") {
    angularDampingOverride=0.995
    angularSpringConstant=200000
    bAngularSpring=true
    bGravityOverrideInSimSpace=true
    bOverrideAngularDamping=true
    bUseGravityOverride=true
    gravityScale=0.35
    simulationSpace=RootRelative
    alphaBoolBlend {blendOption=Linear}
    boundBone {boneName=dyn_R_pouch}
    componentLinearAccScale=<1 1 1>
    componentLinearVelScale=<1 1 1>
    gravityOverride {z=-980}
    physicsBodyDefinitions [{
      sphereCollisionRadius=0
      boundBone {boneName=dyn_R_pouch}
      boxExtents=<80 50 70>
      constraintSetup {angularLimitsMax=<5 4 0>, angularLimitsMin=<-5 -4 0>, angularTarget=<1 0 0>}
      localJointOffset=<-1 -4 -4>
     }]
   }
   n4 = "Constraint - dyn_plateAttachment (PPPPPPPPPPPPPP)"!(
    ComponentPose=@n3
    Alpha=1
    ConstraintWeights_0=0.0065
    ConstraintWeights_1:float=0
    ConstraintWeights_2:float=0
    ConstraintWeights_3:float=0
    ConstraintWeights_4:float=0.0001
    ConstraintWeights_5:float=0
    ConstraintWeights_6:float=0
    ConstraintWeights_7:float=0
    ConstraintWeights_8:float=0.2661
    ConstraintWeights_9:float=0.0164
    ConstraintWeights_10:float=0
    ConstraintWeights_11:float=0.0004
    ConstraintWeights_12:float=0
    ConstraintWeights_13:float=0.7105
   ) -> (Pose) {
    alphaBoolBlend {blendOption=Linear}
    boneToModify {boneName=dyn_plateAttachment}
    constraintSetup [
     {targetBone {boneName=pelvis}}
     {targetBone {boneName=spine_01}}
     {targetBone {boneName=spine_02}}
     {targetBone {boneName=spine_03}}
     {targetBone {boneName=thigh_l}}
     {targetBone {boneName=calf_l}}
     {targetBone {boneName=calf_twist_02_l}}
     {targetBone {boneName=thigh_twist_01_l}}
     {targetBone {boneName=thigh_r}}
     {targetBone {boneName=calf_r}}
     {targetBone {boneName=calf_twist_01_r}}
     {targetBone {boneName=calf_twist_02_r}}
     {targetBone {boneName=foot_r}}
     {targetBone {boneName=thigh_twist_01_r}}
    ]
    constraintWeights [1,1,1,1,1,1,1,1,1,1,1,1,1,1]
   }
   n5 = "Constraint - dyn_R_shoulderPad_base (RR)"!(ComponentPose=@n4, Alpha=1, ConstraintWeights_0=0.7, ConstraintWeights_1:float=0.3) -> (Pose) {
    alphaBoolBlend {blendOption=Linear}
    boneToModify {boneName=dyn_R_shoulderPad_base}
    constraintSetup [{transformType=Rotation, targetBone {boneName=upperarm_r}}, {transformType=Rotation, targetBone {boneName=clavicle_r}}]
    constraintWeights [1,1]
   }
   n6 = Get ADS_pistol Up_R_pointFour() -> (ADS_pistol Up_R_pointFour:double)
   n7 = "Constraint - dyn_R_shoulderPad_base (PP)"!(ComponentPose=@n5, Alpha=@n6, ConstraintWeights_0=0.3, ConstraintWeights_1:float=0.7) -> (Pose) {
    alphaBoolBlend {blendOption=Linear}
    boneToModify {boneName=dyn_R_shoulderPad_base}
    constraintSetup [{targetBone {boneName=clavicle_r}}, {targetBone {boneName=upperarm_twist_01_r}}]
    constraintWeights [1,1]
   }
   n8 = "Transform (Modify) Bone - Bone: dyn_R_shoulderPad_base"(ComponentPose=@n7, Alpha=@n6, Translation="0,-0.5,-3", Rotation="-15,-19,-3", Scale="1,1,1") {
    rotationMode=BMM_Additive
    rotationSpace=BCS_BoneSpace
    scaleMode=BMM_Additive
    scaleSpace=BCS_BoneSpace
    translationMode=BMM_Additive
    translationSpace=BCS_BoneSpace
    alphaBoolBlend {blendOption=Linear}
    boneToModify {boneName=dyn_R_shoulderPad_base}
    rotation=<rot -15 -19 -3>
    translation {y=-0.5, z=-3}
   }
   n9 = "Transform (Modify) Bone - Bone: dyn_R_shoulderPad_base"(ComponentPose=@n8, Alpha=Get pistolDown() -> (pistolDown:double), Translation="0,-1,0.56", Rotation="0,0,0", Scale="1,1.09,1") {
    rotationSpace=BCS_BoneSpace
    scaleMode=BMM_Additive
    scaleSpace=BCS_BoneSpace
    translationMode=BMM_Additive
    translationSpace=BCS_BoneSpace
    alphaBoolBlend {blendOption=Linear}
    alphaScaleBiasClamp {bInterpResult=true, interpSpeedDecreasing=40, interpSpeedIncreasing=40}
    boneToModify {boneName=dyn_R_shoulderPad_base}
    scale {y=1.09}
    translation {y=-1, z=0.56}
   }
   n10 = "Constraint - dyn_R_shoulderPad_base (PP)"(
    ComponentPose=@n9
    Alpha=Get rifleDownConstraintAlpha() -> (rifleDownConstraintAlpha:double)
    ConstraintWeights_0=0.1
    ConstraintWeights_1=0.9
   ) {
    alphaBoolBlend {blendOption=Linear}
    boneToModify {boneName=dyn_R_shoulderPad_base}
    constraintSetup [{targetBone {boneName=clavicle_r}}, {targetBone {boneName=upperarm_twist_01_r}}]
    constraintWeights [1,1]
   }
   n11 = "Constraint - dyn_R_shoulderPad_base (PP)"(ComponentPose=@n10, Alpha=Get isMeleeFloat() -> (isMeleeFloat:double), ConstraintWeights_0=0.15, ConstraintWeights_1=0.85) {
    alphaBoolBlend {blendOption=Linear}
    boneToModify {boneName=dyn_R_shoulderPad_base}
    constraintSetup [{targetBone {boneName=clavicle_r}}, {targetBone {boneName=upperarm_twist_01_r}}]
    constraintWeights [1,1]
   }
   n12 = "Transform (Modify) Bone - Bone: dyn_R_shoulderPad_base"(ComponentPose=@n11, Alpha=Get isMeleeFloat(), Translation="1,-0.5,-2", Rotation="-14,-10,10", Scale="1,1.05,1") {
    rotationMode=BMM_Additive
    rotationSpace=BCS_BoneSpace
    scaleMode=BMM_Additive
    scaleSpace=BCS_BoneSpace
    translationMode=BMM_Additive
    translationSpace=BCS_BoneSpace
    alphaBoolBlend {blendOption=Linear}
    alphaScaleBiasClamp {bInterpResult=true, interpSpeedDecreasing=40, interpSpeedIncreasing=40}
    boneToModify {boneName=dyn_R_shoulderPad_base}
    rotation=<rot -14 -10 10>
    scale {y=1.05}
    translation=<1 -0.5 -2>
   }
   n13 = Get skydivingFloat() -> (skydivingFloat:double)
   n14 = "Constraint - dyn_R_shoulderPad_base (PP)"(ComponentPose=@n12, Alpha=@n13, ConstraintWeights_0=1, ConstraintWeights_1=0) {
    alphaBoolBlend {blendOption=Linear}
    boneToModify {boneName=dyn_R_shoulderPad_base}
    constraintSetup [{targetBone {boneName=upperarm_twist_01_r}}, {targetBone {boneName=clavicle_r}}]
    constraintWeights [1,1]
   }
   n15 = "Transform (Modify) Bone - Bone: dyn_R_shoulderPad"(ComponentPose=@n14, Alpha=@n13, Translation="-5.5,2,-1", Rotation="0,0,20", Scale="1,1,1") {
    rotationMode=BMM_Additive
    rotationSpace=BCS_BoneSpace
    translationMode=BMM_Additive
    translationSpace=BCS_BoneSpace
    alphaBoolBlend {blendOption=Linear}
    boneToModify {boneName=dyn_R_shoulderPad}
    rotation {roll=20}
    translation=<-5.5 2 -1>
   }
   n16 = "Constraint - dyn_L_shoulderPad_base (RR)"!(ComponentPose=@n15, Alpha=1, ConstraintWeights_0=0.7, ConstraintWeights_1:float=0.3) -> (Pose) {
    alphaBoolBlend {blendOption=Linear}
    boneToModify {boneName=dyn_L_shoulderPad_base}
    constraintSetup [{transformType=Rotation, targetBone {boneName=upperarm_l}}, {transformType=Rotation, targetBone {boneName=clavicle_l}}]
    constraintWeights [1,1]
   }
   n17 = "Constraint - dyn_L_shoulderPad_base (PP)"!(
    ComponentPose=@n16
    Alpha=Get ADS_pistol Up_L_pointFour() -> (ADS_pistol Up_L_pointFour:double)
    ConstraintWeights_0=0.7
    ConstraintWeights_1:float=0.3
   ) -> (Pose) {
    alphaBoolBlend {blendOption=Linear}
    boneToModify {boneName=dyn_L_shoulderPad_base}
    constraintSetup [{targetBone {boneName=upperarm_twist_01_l}}, {targetBone {boneName=clavicle_l}}]
    constraintWeights [1,1]
   }
   n18 = "Transform (Modify) Bone - Bone: dyn_L_shoulderPad_base"(ComponentPose=@n17, Alpha=Get ADS_pistol Up_L() -> (ADS_pistol Up_L:double), Translation="-1,1,0.5", Rotation="0,-5,0", Scale="1,0.85,1") {
    rotationMode=BMM_Additive
    rotationSpace=BCS_BoneSpace
    scaleMode=BMM_Additive
    scaleSpace=BCS_BoneSpace
    translationMode=BMM_Additive
    translationSpace=BCS_BoneSpace
    alphaBoolBlend {blendOption=Linear}
    boneToModify {boneName=dyn_L_shoulderPad_base}
    rotation {yaw=-5}
    scale {y=0.85}
    translation=<-1 1 0.5>
   }
   n19 = "Constraint - dyn_L_shoulderPad_base (PP)"(ComponentPose=@n18, Alpha=Get ADS_rifleShotgun Up_L() -> (ADS_rifleShotgun Up_L:double), ConstraintWeights_0=0.1, ConstraintWeights_1=0.9) {
    alphaBoolBlend {blendOption=Linear}
    boneToModify {boneName=dyn_L_shoulderPad_base}
    constraintSetup [{targetBone {boneName=clavicle_l}}, {targetBone {boneName=upperarm_twist_01_l}}]
    constraintWeights [1,1]
   }
   n20 = "Constraint - dyn_L_shoulderPad_base (PP)"(ComponentPose=@n19, Alpha=Get isMeleeFloat(), ConstraintWeights_0=0.3, ConstraintWeights_1=0.7) {
    alphaBoolBlend {blendOption=Linear}
    boneToModify {boneName=dyn_L_shoulderPad_base}
    constraintSetup [{targetBone {boneName=clavicle_l}}, {targetBone {boneName=upperarm_twist_01_l}}]
    constraintWeights [1,1]
   }
   n21 = "Transform (Modify) Bone - Bone: dyn_L_shoulderPad_base"(ComponentPose=@n20, Alpha=Get isMeleeFloat(), Translation="-1,1,2", Rotation="-3,10,-10", Scale="1,1.05,1") {
    rotationMode=BMM_Additive
    rotationSpace=BCS_BoneSpace
    scaleMode=BMM_Additive
    scaleSpace=BCS_BoneSpace
    translationMode=BMM_Additive
    translationSpace=BCS_BoneSpace
    alphaBoolBlend {blendOption=Linear}
    alphaScaleBiasClamp {bInterpResult=true, interpSpeedDecreasing=40, interpSpeedIncreasing=40}
    boneToModify {boneName=dyn_L_shoulderPad_base}
    rotation=<rot -3 10 -10>
    scale {y=1.05}
    translation=<-1 1 2>
   }
   n22 = Get skydivingFloat()
   n23 = "Constraint - dyn_L_shoulderPad_base (PP)"(ComponentPose=@n21, Alpha=@n22, ConstraintWeights_0=1, ConstraintWeights_1=0) {
    alphaBoolBlend {blendOption=Linear}
    boneToModify {boneName=dyn_L_shoulderPad_base}
    constraintSetup [{targetBone {boneName=upperarm_twist_01_l}}, {targetBone {boneName=clavicle_l}}]
    constraintWeights [1,1]
   }
   n24 = "Transform (Modify) Bone - Bone: dyn_L_shoulderPad"(ComponentPose=@n23, Alpha=@n22, Translation="5.5,-2,-1", Rotation="0,0,20", Scale="1,1,1") {
    rotationMode=BMM_Additive
    rotationSpace=BCS_BoneSpace
    translationMode=BMM_Additive
    translationSpace=BCS_BoneSpace
    alphaBoolBlend {blendOption=Linear}
    boneToModify {boneName=dyn_L_shoulderPad}
    rotation {roll=20}
    translation=<5.5 -2 -1>
   }
   n25 = Get Is Playing Emote() -> (Is Playing Emote:double)
   n26 = "Constraint - dyn_R_shoulderPad_base (PP)"(ComponentPose=@n24, Alpha=@n25, ConstraintWeights_0=0.1, ConstraintWeights_1=0.9) {
    alphaBoolBlend {blendOption=Linear}
    boneToModify {boneName=dyn_R_shoulderPad_base}
    constraintSetup [{targetBone {boneName=clavicle_r}}, {targetBone {boneName=upperarm_twist_01_r}}]
    constraintWeights [1,1]
   }
   n27 = "Constraint - dyn_L_shoulderPad_base (PP)"(ComponentPose=@n26, Alpha=@n25, ConstraintWeights_0=0.1, ConstraintWeights_1=0.9) {
    alphaBoolBlend {blendOption=Linear}
    boneToModify {boneName=dyn_L_shoulderPad_base}
    constraintSetup [{targetBone {boneName=clavicle_l}}, {targetBone {boneName=upperarm_twist_01_l}}]
    constraintWeights [1,1]
   }
   n28 = "Transform (Modify) Bone - Bone: dyn_R_scapShoulder_volume"(ComponentPose=@n27, Alpha=Get ADS_pistol Up_R() -> (ADS_pistol Up_R:double), Translation="0,0,0", Rotation="0,0,0", Scale="4,4,4") {
    scaleMode=BMM_Additive
    alphaBoolBlend {blendOption=Linear}
    alphaScaleBiasClamp {bInterpResult=true, interpSpeedDecreasing=30, interpSpeedIncreasing=30}
    boneToModify {boneName=dyn_R_scapShoulder_volume}
    scale=<4 4 4>
   }
   n29 = "Transform (Modify) Bone - Bone: dyn_L_scapShoulder_volume"(ComponentPose=@n28, Alpha=Get ADS_pistol Up_L(), Translation="0,0,0", Rotation="0,0,0", Scale="3,3,3") {
    scaleMode=BMM_Additive
    alphaBoolBlend {blendOption=Linear}
    alphaScaleBiasClamp {bInterpResult=true, interpSpeedDecreasing=30, interpSpeedIncreasing=30}
    boneToModify {boneName=dyn_L_scapShoulder_volume}
    scale=<3 3 3>
   }
   n30 = "Transform (Modify) Bone - Bone: dyn_pocketScoot"(ComponentPose=@n29, Alpha=Get thighRotateAlpha() -> (thighRotateAlpha:double), Translation="0.5,1,-0.5", Rotation="0,0,0", Scale="0.8,0.6,1") {
    scaleMode=BMM_Additive
    translationMode=BMM_Additive
    alphaBoolBlend {blendOption=Linear}
    boneToModify {boneName=dyn_pocketScoot}
    scale=<0.8 0.6>
    translation=<0.5 1 -0.5>
   }
   Output Pose(Result=Layered blend per bone(BasePose=Component To Local(ComponentPose=@n30)) {layerSetup [{branchFilters [{blendDepth=2, boneName=neck_01}, {blendDepth=-1, boneName=head}]}]}) {layerGroup=None}
   Save cached pose 'ShoulderPad_Cache'() {cacheName=ShoulderPad_Cache}
   Save cached pose 'Base_Cache'(Pose=Component To Local()) {cacheName=Base_Cache}
  ]}
 EventGraph {nodes [
   Event BlueprintInitializeAnimation()
   > n1 = Set Skel_Mesh(Skel_Mesh=GetOwnerSkeletalMeshComponent())
   > Set Anim_Inst(Anim_Inst:Anim Instance Object Reference=GetAnimInstance(self=@n1)) -> (Output_Get:Anim Instance Object Reference)
   Event BlueprintUpdateAnimation()
   > n2 = Sequence!(execute) -> (then_0, then_1, then_2:Exec, then_3:Exec, then_4:Exec, then_5:Exec, then_6:Exec, then_7:Exec, then_8:Exec, then_9:Exec)
   > n3 = Branch(Condition=Get skydivingBool() -> (skydivingBool:bool))
   > Set ADS_pistol Up_R(ADS_pistol Up_R:double=0) -> (Output_Get:double)
   > Set pistolDown(pistolDown:double=0) -> (Output_Get:double)
   n4 = GetOwnerSkeletalMeshComponent()
   n5 = BreakRotator(InRot=BreakTransform(InTransform=GetDeltaTransformFromRefPose(self=@n4, BoneName=upperarm_r, BaseName=clavicle_r)).Rotation)
   n6 = MapRangeClamped(Value=@n5.Yaw, InRangeA=40, InRangeB=100)
   Set ADS_pistol Up_R(execute=@n3.else, ADS_pistol Up_R=@n6)
   n7 = GetDeltaTransformFromRefPose(self=@n4, BoneName=clavicle_r, BaseName=spine_05)
   n8 = MapRangeClamped(Value=BreakRotator(InRot=BreakTransform(InTransform=@n7).Rotation).Yaw, InRangeA=0, InRangeB=6)
   > Set pistolDown(pistolDown=@n8)
   > n9 = Branch(Condition=Get isPistol() -> (isPistol:bool))
   > n10 = Set ADS_pistol Up_R(ADS_pistol Up_R=@n6)
   > Set pistolDown(pistolDown=@n8)
   > Set ADS_pistol Up_R_pointFour(ADS_pistol Up_R_pointFour:double=float + float(A=@n10, B=0.4)) -> (Output_Get:double)
   Set ADS_pistol Up_R(execute=@n9.else, ADS_pistol Up_R=0)
   > Set pistolDown(pistolDown=0)
   > Set ADS_pistol Up_R_pointFour(ADS_pistol Up_R_pointFour=0)
   n11 = Branch(execute=@n2.then_1, Condition=Get skydivingBool())
   > Set meleeAlpha(meleeAlpha:double=0) -> (Output_Get:double)
   n12 = MapRangeClamped(Value=@n5.Yaw, InRangeA=30, InRangeB=100, OutRangeA=1, OutRangeB=0)
   Set meleeAlpha(execute=@n11.else, meleeAlpha=@n12)
   > n13 = Branch(Condition=Get isMeleeEquipped() -> (isMeleeEquipped:bool))
   > Set meleeAlpha(meleeAlpha=@n12)
   Set meleeAlpha(execute=@n13.else, meleeAlpha=0)
   Set isMeleeFloat(execute=@n2.then_2, isMeleeFloat:double="To Float (Boolean)"(InBool=Get isMeleeEquipped())) -> (Output_Get:double)
   n14 = OR Boolean!(A=Get bIsSkydiveFloating(), B=Get bIsParachuteOpened(), C:bool=Get bIsSkyDiving()) -> (ReturnValue)
   > Set skydivingBool(skydivingBool:bool=@n14) -> (Output_Get:bool)
   > Set skydivingFloat(skydivingFloat:double="To Float (Boolean)"(InBool=@n14)) -> (Output_Get:double)
   n15 = Get bIsInFrontEnd()
   n16 = Get bIsPlayingEmote()
   > Set Is Playing Emote(Is Playing Emote:double="To Float (Boolean)"(InBool=OR Boolean!(A=@n15, B=@n16) -> (ReturnValue))) -> (Output_Get:double)
   n17 = BreakTransform(InTransform=GetDeltaTransformFromRefPose(self=GetOwnerSkeletalMeshComponent(), BoneName=thigh_l, BaseName=pelvis))
   Set thighRotateAlpha(execute=@n2.then_3, thighRotateAlpha:double=MapRangeClamped(Value=BreakRotator(InRot=@n17.Rotation).Yaw, InRangeA=-70, InRangeB=-90)) -> (Output_Get:double)
   n18 = Branch(execute=@n2.then_4, Condition=Get skydivingBool())
   > Set ADS_rifleShotgun Up_R(ADS_rifleShotgun Up_R:double=0) -> (Output_Get:double)
   Set ADS_rifleShotgun Up_R(execute=@n18.else, ADS_rifleShotgun Up_R=@n12)
   > n19 = Branch(Condition=Get isRifleShotgun() -> (isRifleShotgun:bool))
   > Set ADS_rifleShotgun Up_R(ADS_rifleShotgun Up_R=@n12)
   Set ADS_rifleShotgun Up_R(execute=@n19.else, ADS_rifleShotgun Up_R=0)
   n20 = Cast To FortPlayerAnimInstance(execute=@n2.then_5, Object=GetAnimInstance(self=GetOwnerSkeletalMeshComponent()))
   > n21 = Switch on EFortWeaponCoreAnimation!(execute, Selection:EFortWeaponCoreAnimation Enum=Get WeaponCoreAnim(self=@n20)) -> (
    Melee:Exec
    Pistol:Exec
    Shotgun:Exec
    PaperBlueprint:Exec
    Rifle:Exec
    MeleeOneHand:Exec
    MachinePistol:Exec
    RocketLauncher:Exec
    GrenadeLauncher:Exec
    GoingCommando:Exec
    AssaultRifle:Exec
    TacticalShotgun:Exec
    SniperRifle:Exec
    TrapPlacement:Exec
    ShoulderLauncher:Exec
    AbilityDecoTool:Exec
    Crossbow:Exec
    C4:Exec
    RemoteControl:Exec
    DualWield:Exec
    AR_BullPup:Exec
    AR_ForwardGrip:Exec
    MedPackPaddles:Exec
    SMG_P90:Exec
    AR_DrumGun:Exec
    Consumable_Small:Exec
    Consumable_Large:Exec
    Balloon:Exec
    MountedTurret:Exec
    CreativeTool:Exec
    ExplosiveBow:Exec
    AshtonIndigo:Exec
    AshtonChicago:Exec
    MeleeDualWield:Exec
    Unarmed:Exec
   )
   Set isRifleShotgun(execute=[@n21.AssaultRifle, @n21.Rifle, @n21.Shotgun, @n21.SniperRifle, @n21.TacticalShotgun], isRifleShotgun:bool=true) -> (Output_Get:bool)
   > Set isMeleeEquipped(isMeleeEquipped:bool=false) -> (Output_Get:bool)
   > Set isPistol(isPistol=false)
   > Set rifleDownConstraintAlpha(rifleDownConstraintAlpha:double=0.4) -> (Output_Get:double)
   Set isPistol(execute=[@n21.MachinePistol, @n21.Pistol], isPistol=true)
   > Set isMeleeEquipped(isMeleeEquipped=false)
   > Set isRifleShotgun(isRifleShotgun=false)
   Set isMeleeEquipped(execute=[@n21.Melee, @n21.MeleeDualWield, @n21.MeleeOneHand], isMeleeEquipped=true)
   > Set isRifleShotgun(isRifleShotgun=false)
   > Set isPistol(isPistol=false)
   Branch(execute=@n2.then_6, Condition=OR Boolean(A=@n16, B=Get bIsSkyDiving()))
   > Set isPistol(isPistol=false)
   > Set isRifleShotgun(isRifleShotgun=false)
   > Set isMeleeEquipped(isMeleeEquipped=false)
   n22 = Branch(execute=@n2.then_7, Condition=Get skydivingBool())
   > Set ADS_pistol Up_L(ADS_pistol Up_L:double=0) -> (Output_Get:double)
   n23 = BreakRotator(InRot=BreakTransform(InTransform=GetDeltaTransformFromRefPose(self=@n4, BoneName=upperarm_l, BaseName=clavicle_l)).Rotation)
   n24 = MapRangeClamped(Value=@n23.Yaw, InRangeA=30, InRangeB=70)
   Set ADS_pistol Up_L(execute=@n22.else, ADS_pistol Up_L=@n24)
   > n25 = Branch(Condition=Get isPistol())
   > n26 = Set ADS_pistol Up_L(ADS_pistol Up_L=@n24)
   > Set ADS_pistol Up_L_pointFour(ADS_pistol Up_L_pointFour:double=float + float(A=@n26, B=0.4)) -> (Output_Get:double)
   Set ADS_pistol Up_L(execute=@n25.else, ADS_pistol Up_L=0)
   > Set ADS_pistol Up_L_pointFour(ADS_pistol Up_L_pointFour=0)
   n27 = Branch(execute=@n2.then_8, Condition=Get skydivingBool())
   > Set ADS_rifleShotgun Up_L(ADS_rifleShotgun Up_L:double=0) -> (Output_Get:double)
   n28 = MapRangeClamped(Value=@n23.Yaw, InRangeA=30, InRangeB=100, OutRangeA=1, OutRangeB=0)
   Set ADS_rifleShotgun Up_L(execute=@n27.else, ADS_rifleShotgun Up_L=@n28)
   > n29 = Branch(Condition=Get isRifleShotgun())
   > Set ADS_rifleShotgun Up_L(ADS_rifleShotgun Up_L=@n28)
   Set ADS_rifleShotgun Up_L(execute=@n29.else, ADS_rifleShotgun Up_L=0)
   Reroute Node()
   n30 = Get Anim_Inst() -> (Anim_Inst:Anim Instance Object Reference)
   > Is Valid(InputObject=@n30)
   > Branch(Condition=Get bIsInFrontEnd())
   n31 = Get bIsPlayingEmote()
   > Set FrontEnd_Idle_On(FrontEnd_Idle_On:bool=AND Boolean(
     A=Montage_IsPlaying(self=@n30, Montage=/BRCosmetics/Animation/Game/MainPlayer/Menu/BR/CustomSkins/_s19/RustyBoltM/RustyBolt_Male_Idle_Montage)
     B=NOT Boolean(A=@n31)
    )) -> (Output_Get:bool)
  ]}
]
variables [
 ADS_pistol Up_L {type=number}
 ADS_pistol Up_L_pointFour {type=number}
 ADS_pistol Up_R {type=number}
 ADS_pistol Up_R_pointFour {type=number}
 ADS_rifleShotgun Up_L {type=number}
 ADS_rifleShotgun Up_R {type=number}
 Anim_Inst {type=AnimInstance, default=null}
 FrontEnd_Idle_On {type=boolean}
 Is Playing Emote {type=number}
 isMeleeEquipped {type=boolean}
 isMeleeFloat {type=number}
 isPistol {type=boolean}
 isRifleShotgun {type=boolean}
 meleeAlpha {type=number}
 pistolDown {type=number}
 rifleDownConstraintAlpha {type=number}
 Rotation {type=Rotator, default=<rot -15 -19 -3>}
 Skel_Mesh {type=SkeletalMeshComponent}
 skydivingBool {type=boolean}
 skydivingFloat {type=number}
 thighRotateAlpha {type=number}
 Translation {type=Vector, default=<0 -0.5 -3>}
]
